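# Niagara effect for destroying a DesertPoint tree, throwing out mesh chunks, debris, sparks and tinted smoke, largely in pink and purple.
# All six emitters spawn on the surface of the mesh supplied through User.CrystalStaticMesh.
# CPU mesh bursts use four spawn groups with group-masked collision, alongside two CPU smoke layers.
# Three GPU sprite emitters add colliding debris, color-shifting sparks and noise-driven motion, with system distance-scale variables adjusting several sizes and speeds.
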NiagaraSystem /Game/Effects/Fort_Effects/Effects/Environment/DesertPoint/NS_DesertPoint_Tree_Destruction_M
dependencies {
 usedDataInterfaces [&NiagaraDataInterfaceCurve, &NiagaraDataInterfaceColorCurve, &NiagaraDataInterfaceVectorCurve]
 usedDynamicInputs [/Game/Effects/Niagara/DynamicInputs/Math/FN_Fit, &MakeIntFromFloat]
 usedModules [
  &self:SystemSpawnScript.SystemScriptSource.SystemScriptGraph.NiagaraNodeAssignment_0.SetVariables_#1
  &SystemState
 ]
}
emitters {
 DirectionalBurstMeshes_01 {
  emitterUpdateScript [
   &EmitterState(Life Cycle Mode=::NewEnumerator1, Loop Behavior=::NewEnumerator1, Loop Duration=1, UseLoopDelay=false)
   &SpawnBurst_Instantaneous(Spawn Count=&UniformRangedInt(Minimum=10, Maximum=15), Use Spawn Probability=false)
   &SpawnBurst_Instantaneous(Spawn Count=&UniformRangedInt(Minimum=15, Maximum=20), Spawn Group=1, Use Spawn Probability=false) {module=SpawnBurst_Instantaneous001}
   &SpawnBurst_Instantaneous(Spawn Count=&UniformRangedInt(Minimum=10, Maximum=25), Spawn Group=2, Use Spawn Probability=false) {module=SpawnBurst_Instantaneous002}
   &SpawnBurst_Instantaneous(Spawn Count=&UniformRangedInt(Minimum=20, Maximum=25), Spawn Group=3, Use Spawn Probability=false) {module=SpawnBurst_Instantaneous003}
  ]
  particleSpawnScript [
   /SetVariables(Particles.RandonInt2:NiagaraInt32=&UniformRangedInt(Minimum=0, Maximum=3)) {isSetParameters=true}
   /Niagara/Modules/Spawn/Initialization/InitializeParticle(
    Lifetime=&RandomRangeFloat(Minimum=3, Maximum=5)
    Mass=&UniformRangedFloat(Minimum=250000, Maximum=250000)
    Mesh Scale=&Multiply_Vector(
     A=&UniformRangedVector(Minimum=<0.2 0.2 0.2>, Maximum=<0.3 0.3 0.3>)
     B=&Multiply_VectorByFloat(Vector=&VectorFromFloat(Value=$System.ScaleDistance), Float=$System.MeshScale)
    )
    Module.Write Color=true
    Module.Write Lifetime=true
    Module.Write Mass=true
    Module.Write Material Random=true
    Module.Write Position=false
    Module.Write Scale=true
    Module.Write SpriteRotation=false
    Module.Write SpriteSize=false
    Position=&SimulationPosition()
   )
   &AddVelocityInCone(
    Cone Angle=90
    Cone Axis=<0 0 1>
    Module.Use Velocity Falloff On Cone Axis=true
    Velocity Falloff Away From Cone Axis=0.333
    Velocity Strength=&UniformRangedFloat(
     Minimum=&Multiply_Float(A=250, B=$System.ScaleDistanceSpeed)
     Maximum=&Multiply_Float(A=500, B=$System.ScaleDistanceSpeed)
    )
   )
   /Niagara/Modules/Spawn/MeshInterface/SampleStaticMesh(OverrideRandomSeed=false, StaticMesh:NiagaraDataInterfaceStaticMesh=$User.CrystalStaticMesh)
   /Niagara/Modules/Spawn/Location/StaticMeshLocation(Module.Enable Axis Constraint=false)
   &AddVelocityFromPoint(
    Module.Use Falloff Distance=false
    Origin Offset=&UniformRangedVector(Minimum=<-1 -1 0>, Maximum=<1 1 0>)
    Origin Offset Coordinate Space=::Simulation
    Velocity Strength=&UniformRangedFloat(Minimum=200, Maximum=400)
   )
   /Niagara/Modules/Spawn/Orientation/V2/InitialMeshOrientation(
    Orientation Axis:Vector3f=<0 0 1>
    Orientation Vector=<0 0 1>
    Rotation=&UniformRangedVector(Minimum=<0 0 0>, Maximum=<1 1 1>)
    Use Orientation Vector=true
    Use Rotation Vector=true
   )
   &AddRotationalVelocity(Rotation Rate=&UniformRangedVector(Minimum=<1 1 1>, Maximum=<-1 -1 -1>))
   &ApplyInitialForces()
  ]
  particleUpdateScript [
   &ParticleState()
   &GravityForce(Gravity=<0 0 -1600>)
   &Drag(
    Drag=&FloatFromCurve(FloatCurve=DataInterface() {propertyValues {Curve {keys [
         <at 0 0>
         {arriveTangent=-4.1608, interpMode=RCIM_Cubic, leaveTangent=-4.1608, tangentMode=RCTM_User, time=0.25, value=1}
         {arriveTangent=-0.1083, interpMode=RCIM_Cubic, leaveTangent=-0.1083, tangentMode=RCTM_User, time=1, value=0}
        ]}}})
    Use Linear Drag=true
    Use Rotational Drag=false
   ) {enabled=false}
   &ScaleColor(
    Scale Alpha=&FloatFromCurve(FloatCurve=DataInterface() {propertyValues {Curve {keys [<at -0.0046 1.0113 RCIM_Cubic>,<at 1.0011 0.002 RCIM_Cubic>]}}})
    ScaleA=true
    ScaleRGB=true
   )
   &Collision(
    Advanced Aging Rate=1
    Age Colliding Particles=false
    Collision Enabled=/Niagara/DynamicInputs/Helpers/MaskBoolBySpawnGroup(
     Masked Bool:NiagaraBool=false
     Passthrough Bool:NiagaraBool=&SetBoolByFloatComparison(A=$Particles.Age, B=0)
     Spawn Group Mask:NiagaraInt32=0
    )
    Control Roll On Collision=true
    Mesh Dimensions:Vector3f=<150 150 150>
    Module.Control Rest State Via Surface Normal=false
    Module.EnableMaxCollisionCount=false
    Module.Randomize Collision Normal Vector=true
    Radius Calculation Type=::NewEnumerator1
    Restitution=0.35
    ScaleRateOfRotation=false
   )
   &SolveForcesAndVelocity(Clamp Velocity=false, Limit Acceleration=false)
   &MeshRotationForce(Rotation=<0 0 1>) {enabled=false}
   &SolveRotationalForcesAndVelocity()
   &ScaleMeshSize(Scale Factor=&VectorFromFloat(Value=&FloatFromCurve(FloatCurve=DataInterface() {propertyValues {Curve {keys [<at 0 1>,<at 0.95 1>,<at 1 0>]}}})))
  ]
  rendererClasses [&NiagaraMeshRendererProperties]
  renderers [
   {rendererClass=&NiagaraMeshRendererProperties, rendererIndex=0}
   {rendererClass=&NiagaraMeshRendererProperties, rendererIndex=1}
   {rendererClass=&NiagaraMeshRendererProperties, rendererIndex=2}
   {rendererClass=&NiagaraMeshRendererProperties, rendererIndex=3}
  ]
 }
 HT_Debris {
  simTarget=GPUComputeSim
  emitterUpdateScript [
   &EmitterScalability(
    Module.bApplySpawnCountScale=false
    Module.bApplySpawnCountScaleByDistanceFraction=false
    Module.bUseMaxDistance=true
    Module.bUseMinDistance=false
   ) {enabled=false}
   &EmitterState(Life Cycle Mode=::NewEnumerator1, Loop Behavior=::NewEnumerator1, Loop Duration=0.2, UseLoopDelay=false)
   &EmitterLifeCycle(MaxLoopCount=1, NextLoopDuration=0.1) {enabled=false}
   &SpawnRate(
    SpawnRate=&Multiply_Float(
     A=1500
     B=&FloatFromCurve(FloatCurve=DataInterface() {propertyValues {Curve {keys [<at 0.2236 1.0023>,<at 1 0>]}}}, CurveIndex=$Emitter.NormalizedLoopAge)
    )
    Use Spawn Probability=false
   )
  ]
  particleSpawnScript [
   /SetVariables(Particles.RandomFloat=&UniformRangedFloat(Minimum=0, Maximum=3)) {isSetParameters=true}
   &InitializeParticle(
    Color=&UniformRangedLinearColor(Minimum=<rgba 1.9339 0.3786 2 1>, Maximum=<rgba 2 1.3104 1.7856 1>)
    Lifetime=&UniformRangedFloat(Minimum=0.5, Maximum=1.5)
    Module.Write Color=true
    Module.Write Lifetime=true
    Module.Write Mass=true
    Module.Write Material Random=true
    Module.Write Position=false
    Module.Write Scale=false
    Module.Write SpriteRotation=true
    Module.Write SpriteSize=true
    Position=&SimulationPosition()
    Sprite Rotation=&UniformRangedFloat(Minimum=0, Maximum=360)
    Sprite Size=&Vector2DFromFloat(Value=&Multiply_Float(A=&UniformRangedFloat(Minimum=5, Maximum=10), B=$System.ScaleDistance))
   )
   &SampleStaticMesh(OverrideRandomSeed=false, StaticMesh:NiagaraDataInterfaceStaticMesh=$User.CrystalStaticMesh)
   &StaticMeshLocation(Module.Enable Axis Constraint=false)
   &AddVelocityInCone(
    Cone Angle=80
    Cone Axis=&RotateVectorByQuaternion(Vector To Rotate=<0 0 1>, Quaternion=<0 0 0 1>)
    Cone Axis Coordinate Space=::Simulation
    Module.Use Velocity Falloff On Cone Axis=true
    Velocity Falloff Away From Cone Axis=0.333
    Velocity Strength=&UniformRangedFloat(Minimum=500, Maximum=1000)
   ) {module=AddVelocityInCone001}
   &PointForce(
    Force Origin=&SimulationPosition()
    Force Strength=&UniformRangedFloat(
     Minimum=100
     Maximum=&Multiply_Float(
      A=250
      B=&FloatFromCurve(FloatCurve=DataInterface() {propertyValues {Curve {keys [<at 0.0228 1.5928>,<at 0.9916 -1.2016>]}}}, CurveIndex=$Emitter.NormalizedLoopAge)
     )
    )
    Module.Use Falloff=false
    Module.UseLinearForce=false
   )
   &ApplyInitialForces()
   &AddVelocityInCone(
    Cone Axis=<0 0 1>
    Module.Use Velocity Falloff On Cone Axis=false
    Velocity Strength=&UniformRangedFloat(Minimum=100, Maximum=400)
   ) {enabled=false}
   /Niagara/Modules/Update/SubUV/SubUVAnimation(Random SubUV Lookup Index:NiagaraFloat=$Particles.RandomFloat, SubUV Animation Mode=::NewEnumerator1) {module=SubUVAnimation001}
  ]
  particleUpdateScript [
   &ParticleState() {module=UpdateAge}
   &Drag(Use Linear Drag=true, Use Rotational Drag=false)
   &AccelerationForce(Acceleration=<0 0 -980>) {enabled=false}
   &GravityForce()
   &Collision(
    Advanced Aging Rate=1
    Age Colliding Particles=false
    GPU Collision Type=::NewEnumerator2
    Module.Control Rest State Via Surface Normal=false
    Module.EnableMaxCollisionCount=false
    Module.Randomize Collision Normal Vector=true
    Particle Radius Scale=0.5
    Randomize Collision Normal=0.25
    Rest State Time Range=1
    Restitution=0.5
   )
   &SolveForcesAndVelocity(Clamp Velocity=false, Limit Acceleration=false)
   &SubUVAnimation(SubUV Animation Mode=::NewEnumerator1) {enabled=false}
   &SpriteRotationRate(Rotation Rate=&Multiply_Float(A=$Particles.RandomFloat, B=4)) {enabled=false}
   &Color() {enabled=false}
   &ScaleSpriteSize(Scale Factor=&Vector2DFromFloat(Value=&FloatFromCurve(FloatCurve=DataInterface() {propertyValues {Curve {keys [<at 0 0>,<at 0.1 1>,<at 0.6514 0.9748>,<at 1 0>]}}}, Scale Curve=2)))
   &ScaleColor(
    Scale Mode=::NewEnumerator1
    Scale RGBA=&Vector4FromCurve(Vector4Curve=DataInterface() {propertyValues {
       AlphaCurve {keys [<at 0 1>,<at 1 0>]}
       BlueCurve {keys [<at 0 1>,<at 1 1>]}
       GreenCurve {keys [<at 0 1>,<at 1 1>]}
       RedCurve {keys [<at 0 1>,<at 1 1>]}
      }})
    ScaleRGBA=true
   )
  ]
  rendererClasses [&NiagaraSpriteRendererProperties]
  renderers [{rendererClass=&NiagaraSpriteRendererProperties}]
 }
 Pop_Sparks {
  simTarget=GPUComputeSim
  emitterUpdateScript [
   &EmitterState()
   &SpawnRate(SpawnRate=500, Use Spawn Probability=true)
  ]
  particleSpawnScript [
   &InitializeParticle(
    Color=&LerpLinearColors(StartColor=<rgba 1 1 1 1>, EndColor=<rgba 1 1 1 1>, LerpFactor=1)
    Lifetime=&Multiply_Float(A=&UniformRangedFloat(Minimum=0.1, Maximum=2), B=1)
    Module.Write Color=true
    Module.Write Lifetime=true
    Module.Write Mass=true
    Module.Write Material Random=true
    Module.Write Position=false
    Module.Write Scale=false
    Module.Write SpriteRotation=false
    Module.Write SpriteSize=true
    Position=&SimulationPosition()
    Sprite Size=&UniformRangedVector2D(
     Minimum=&Vector2DFromFloat(Value=&Multiply_Float(A=50, B=$System.ScaleDistance))
     Maximum=&Vector2DFromFloat(Value=&Multiply_Float(A=150, B=$System.ScaleDistance))
    )
   )
   &AddVelocityFromPoint(
    Module.Use Falloff Distance=false
    Origin Offset=&UniformRangedVector(Minimum=<-1 -1 0>, Maximum=<1 1 0>)
    Origin Offset Coordinate Space=::Simulation
    Velocity Strength=&UniformRangedFloat(Minimum=100, Maximum=200)
   )
   &SampleStaticMesh(OverrideRandomSeed=false, StaticMesh:NiagaraDataInterfaceStaticMesh=$User.CrystalStaticMesh)
   &StaticMeshLocation(Module.Enable Axis Constraint=false)
   /Niagara/Modules/Update/Forces/CurlNoiseForce(Noise Frequency=0.25, Noise Strength=25, Pan Noise=true, Pan Noise Field=<0 0 -5>) {module=CurlNoiseForce001}
   &ApplyInitialForces()
  ]
  particleUpdateScript [
   &ParticleState()
   /SetVariables(Particles.GradientColor:Vector3f=&MakeVectorFromLinearColorRGB(Color=&MakeLinearColorFromVectorAndFloat(
      "Vector (RGB)"=&MakeVectorFromLinearColorRGB(Color=&ColorFromCurve(ColorCurve=DataInterface() {propertyValues {
          AlphaCurve {keys [<at -0.0023 1 RCIM_Cubic>,<at 0.9977 1 RCIM_Cubic>]}
          BlueCurve {keys [<at 0.1175 0.3005 RCIM_Cubic>,<at 0.2682 0.2705>,<at 0.4507 0.2122>,<at 0.5951 0.5271>,<at 0.7691 1.5>,<at 0.9071 0.521 RCIM_Cubic>]}
          GreenCurve {keys [<at 0.1175 0.1878 RCIM_Cubic>,<at 0.2682 0.2623>,<at 0.4507 0.2346>,<at 0.5951 0.4287>,<at 0.7691 0.4532>,<at 0.9071 0.4851 RCIM_Cubic>]}
          RedCurve {keys [<at 0.1175 0.3372 RCIM_Cubic>,<at 0.2682 0.7305>,<at 0.4507 0.7758>,<at 0.5951 0.7084>,<at 0.7691 0>,<at 0.9071 0.807 RCIM_Cubic>]}
         }}))
      "Float (Alpha)"=1
     ))) {isSetParameters=true}
   /SetVariables(Particles.Color=&MakeLinearColorFromVectorAndFloat(
     "Vector (RGB)"=&Lerp_Vector(
      A=&Multiply_VectorByFloat(Vector=$Particles.GradientColor, Float=5)
      B=&Multiply_VectorByFloat(Vector=$Particles.GradientColor, Float=25)
      Alpha=0.5
     )
     "Float (Alpha)"=&FloatFromCurve(
      FloatCurve=DataInterface() {propertyValues {Curve {keys [
          {interpMode=RCIM_Cubic, tangentMode=RCTM_User, time=-0.0063, value=0.0184}
          {arriveTangent=-0.8483, interpMode=RCIM_Cubic, leaveTangent=-0.8483, tangentMode=RCTM_User, time=0.1513, value=1.0122}
          {arriveTangent=-5.0978, interpMode=RCIM_Cubic, leaveTangent=-5.0978, tangentMode=RCTM_User, time=0.2818, value=1.2143}
          {arriveTangent=-0.2783, interpMode=RCIM_Cubic, leaveTangent=-0.2783, tangentMode=RCTM_User, time=0.3901, value=0.5412}
          {arriveTangent=0.3112, interpMode=RCIM_Cubic, leaveTangent=0.3112, tangentMode=RCTM_User, time=0.5407, value=1.3419}
          {arriveTangent=-3.7119, interpMode=RCIM_Cubic, leaveTangent=-3.7119, tangentMode=RCTM_User, time=0.7242, value=0.3749}
          {interpMode=RCIM_Cubic, tangentMode=RCTM_User, time=1.0011, value=0.002}
         ]}}}
      Scale Curve=0.8
     )
    )) {isSetParameters=true}
   &CurlNoiseForce(
    Noise Frequency=1
    Noise Strength=&FloatFromCurve(FloatCurve=DataInterface() {propertyValues {Curve {keys [<at 0 1>,<at 0.25 0>]}}}, Scale Curve=250)
    Pan Noise=true
    Pan Noise Field=<0 10 0>
   )
   &Drag(Drag=&UniformRangedFloat(Minimum=0, Maximum=0), Use Linear Drag=true, Use Rotational Drag=false)
   &GravityForce(Gravity=<0 0 -400>)
   &Collision(
    Age Colliding Particles=false
    GPU Collision Type=::NewEnumerator2
    Module.Control Rest State Via Surface Normal=false
    Module.EnableMaxCollisionCount=false
    Module.Randomize Collision Normal Vector=false
    Particle Radius Scale=15
    Restitution=0.2
   ) {enabled=false}
   &SolveForcesAndVelocity(Clamp Velocity=false, Limit Acceleration=false)
   &ScaleSpriteSize(Scale Factor=&Vector2DFromFloat(Value=&FloatFromCurve(FloatCurve=DataInterface() {propertyValues {Curve {keys [
          {arriveTangent=7.1136, interpMode=RCIM_Cubic, leaveTangent=7.1136, tangentMode=RCTM_User, time=0, value=0}
          {arriveTangent=1.4654, interpMode=RCIM_Cubic, leaveTangent=1.4654, time=0.18, value=1}
          <at 0.3814 0.5589>
          {arriveTangent=-0.0952, interpMode=RCIM_Cubic, leaveTangent=-0.0952, tangentMode=RCTM_User, time=0.7849, value=0.7663}
          <at 1 0.5 RCIM_Cubic>
         ]}}})))
   &ScaleSpriteSizeBySpeed(Max Scale Factor=<0.5 0.5>)
  ]
  rendererClasses [&NiagaraSpriteRendererProperties]
  renderers [{rendererClass=&NiagaraSpriteRendererProperties}]
 }
 SlurpWispySmoke {
  emitterUpdateScript [
   &EmitterState(Life Cycle Mode=::NewEnumerator1, Loop Behavior=::NewEnumerator1, Loop Duration=1, UseLoopDelay=false)
   &SpawnBurst_Instantaneous(Spawn Count=&UniformRangedInt(Minimum=8, Maximum=8), Use Spawn Probability=false)
  ]
  particleSpawnScript [
   &InitializeParticle(
    Color=&UniformRangedLinearColor(Minimum=<rgba 0.5 0.5 0.5 0.4>, Maximum=<rgba 1 1 1 0.6>)
    Lifetime=&UniformRangedFloat(Minimum=1, Maximum=2)
    Module.Write Color=true
    Module.Write Lifetime=true
    Module.Write Mass=true
    Module.Write Material Random=true
    Module.Write Position=true
    Module.Write Scale=false
    Module.Write SpriteRotation=true
    Module.Write SpriteSize=true
    Position=&Add_Vector(A=Linked() {linkedVariable {name=Engine.Owner.Position, type=NiagaraPosition}}, B=<0 0 50>)
    Sprite Rotation=&UniformRangedFloat(Minimum=0, Maximum=360)
    Sprite Size=&Vector2DFromFloat(Value=&UniformRangedFloat(Minimum=200, Maximum=400))
   )
   &AddVelocityInCone(
    Cone Angle=120
    Cone Axis=<-1 0 0>
    Module.Use Velocity Falloff On Cone Axis=false
    Velocity Strength=&UniformRangedFloat(Minimum=500, Maximum=800)
   )
   &SampleStaticMesh(OverrideRandomSeed=false, StaticMesh:NiagaraDataInterfaceStaticMesh=$User.CrystalStaticMesh)
   &StaticMeshLocation(Module.Enable Axis Constraint=false)
   &Color(Color=&MultiplyLinearColor(
     Linear Color=$Particles.Color
     Scale Factor=&MakeVector4_FromLinearColor(LinearColor=&UniformRangedLinearColor(Minimum=<rgba 1 0.35 0.7266 1>, Maximum=<rgba 0.876 0.5 1 1>))
    ))
   /SetVariables(Particles.UVScale:Vector2f=&UniformRangedVector2D(Minimum=<-1 1>, Maximum=<-1 1>)) {isSetParameters=true, module=SetVariables_#2}
   &CameraOffset(Camera Offset Amount=100)
  ]
  particleUpdateScript [
   &ParticleState()
   &Drag(Drag=&UniformRangedFloat(Minimum=2, Maximum=3), Use Linear Drag=true, Use Rotational Drag=false)
   &GravityForce(Gravity=<0 0 -650>)
   &SolveForcesAndVelocity(Clamp Velocity=false, Limit Acceleration=false)
   &SpriteRotationRate(Rotation Rate=&Multiply_Float(
     A=&Lerp_Float(A=400, B=-400, Alpha=$Particles.Initial.MaterialRandom)
     B=&Lerp_Float(A=1, B=0, Alpha=$Particles.NormalizedAge)
    )) {enabled=false}
   &ScaleColor(
    Scale Alpha=&FloatFromCurve(
     FloatCurve=DataInterface() {propertyValues {Curve {keys [
         {interpMode=RCIM_Cubic, tangentMode=RCTM_User, time=0, value=1}
         {arriveTangent=-1, interpMode=RCIM_Cubic, leaveTangent=-1, tangentMode=RCTM_User, time=0.0954, value=0.4177}
         <at 0.9979 0.0215>
        ]}}}
     Scale Curve=&Multiply_Float(A=2, B=$System.ScaleDistance)
    )
    Scale RGB=<1 1 3>
    ScaleA=true
    ScaleRGB=true
   )
   &ScaleSpriteSize(Scale Factor=&Vector2DFromFloat(Value=&Lerp_Float(A=1, B=2, Alpha=$Particles.NormalizedAge)))
   &SubUVAnimation(Number Of Frames=63)
   &DynamicMaterialParameters(
    Index 0 Param 1=1
    Index 0 Param 2=0.5
    Index 0 Param 3=1
    Index 0 Param 4=1
    Module.Param0WriteEnabled=true
    Module.Param1WriteEnabled=true
    Module.Param2WriteEnabled=true
    Module.Param3WriteEnabled=true
   )
  ]
  rendererClasses [&NiagaraSpriteRendererProperties]
  renderers [{rendererClass=&NiagaraSpriteRendererProperties}]
 }
 SlurpWispySmoke001 {
  emitterUpdateScript [
   &EmitterState(Life Cycle Mode=::NewEnumerator1, Loop Behavior=::NewEnumerator1, Loop Duration=1, UseLoopDelay=false)
   &SpawnBurst_Instantaneous(
    Spawn Count=&Multiply_Int(A=&UniformRangedInt(Minimum=4, Maximum=4), B=$System.SpawnRateScaleDist)
    Use Spawn Probability=false
   )
  ]
  particleSpawnScript [
   &InitializeParticle(
    Color=&UniformRangedLinearColor(Minimum=<rgba 0.0812 0.1267 1 0.35>, Maximum=<rgba 0.8921 0.6033 1 0.5>)
    Lifetime=&UniformRangedFloat(Minimum=0.5, Maximum=1)
    Module.Write Color=true
    Module.Write Lifetime=true
    Module.Write Mass=true
    Module.Write Material Random=true
    Module.Write Position=true
    Module.Write Scale=false
    Module.Write SpriteRotation=true
    Module.Write SpriteSize=true
    Position=&Add_Vector(A=Linked() {linkedVariable {name=Engine.Owner.Position, type=NiagaraPosition}}, B=<0 0 50>)
    Sprite Rotation=&UniformRangedFloat(Minimum=0, Maximum=360)
    Sprite Size=&Vector2DFromFloat(Value=&Multiply_Float(A=&UniformRangedFloat(Minimum=600, Maximum=800), B=$System.ScaleDistance))
   )
   &SystemLocation(MaskSpawnGroup=false, Offset=$System.LocationOffset) {enabled=false}
   &SampleStaticMesh(OverrideRandomSeed=false, StaticMesh:NiagaraDataInterfaceStaticMesh=$User.CrystalStaticMesh)
   &StaticMeshLocation(Module.Enable Axis Constraint=false)
   &Color(Color=&MultiplyLinearColor(
     Linear Color=$Particles.Color
     Scale Factor=&MakeVector4_FromLinearColor(LinearColor=&UniformRangedLinearColor(Minimum=<rgba 1 0.75 0.9935 1>, Maximum=<rgba 0.938 0.75 1 1>))
    ))
   /SetVariables(Particles.UVScale:Vector2f=&UniformRangedVector2D(Minimum=<-1 1>, Maximum=<-1 1>)) {isSetParameters=true, module=SetVariables_#2}
   &CameraOffset(Camera Offset Amount=100)
  ]
  particleUpdateScript [
   &ParticleState()
   &Drag(Drag=&UniformRangedFloat(Minimum=2, Maximum=3), Use Linear Drag=true, Use Rotational Drag=false)
   &GravityForce(Gravity=<0 0 -650>)
   &SolveForcesAndVelocity(Clamp Velocity=false, Limit Acceleration=false)
   &SpriteRotationRate(Rotation Rate=&Multiply_Float(
     A=&Lerp_Float(A=400, B=-400, Alpha=$Particles.Initial.MaterialRandom)
     B=&Lerp_Float(A=1, B=0, Alpha=$Particles.NormalizedAge)
    )) {enabled=false}
   &ScaleColor(
    Scale Alpha=&FloatFromCurve(
     FloatCurve=DataInterface() {propertyValues {Curve {keys [
         {interpMode=RCIM_Cubic, tangentMode=RCTM_User, time=0.0026, value=0.1682}
         <at 0.0461 0.929>
         {arriveTangent=-1, interpMode=RCIM_Cubic, leaveTangent=-1, tangentMode=RCTM_User, time=0.4511, value=0.5257}
         <at 1 0>
        ]}}}
     Scale Curve=&Multiply_Float(A=1, B=$System.ScaleDistance)
    )
    Scale RGB=&Multiply_VectorByFloat(
     Vector=<2 1 4>
     Float=&FloatFromCurve(FloatCurve=DataInterface() {propertyValues {Curve {keys [
          {interpMode=RCIM_Cubic, tangentMode=RCTM_User, time=0, value=15}
          {arriveTangent=-15, interpMode=RCIM_Cubic, leaveTangent=-15, tangentMode=RCTM_User, time=0.1096, value=7.1975}
          <at 1 0>
         ]}}})
    )
    ScaleA=true
    ScaleRGB=true
   )
   &ScaleSpriteSize(Scale Factor=&Vector2DFromFloat(Value=&FloatFromCurve(
      FloatCurve=DataInterface() {propertyValues {Curve {keys [
          {arriveTangent=1.4022, interpMode=RCIM_Cubic, leaveTangent=1.4022, tangentMode=RCTM_User, time=-0.0054, value=0.6171}
          {arriveTangent=0.5, interpMode=RCIM_Cubic, leaveTangent=0.5, tangentMode=RCTM_User, time=0.2875, value=0.8966}
          <at 1 1>
         ]}}}
      Scale Curve=2
     )))
   &DynamicMaterialParameters(
    Index 0 Param 1=&RandomRangeFloat(Minimum=0, Maximum=1)
    Index 0 Param 2=&RandomRangeFloat(Minimum=0, Maximum=1)
    Index 0 Param 3=0.25
    Index 0 Param 4=1
    Module.Param0WriteEnabled=true
    Module.Param1WriteEnabled=true
    Module.Param2WriteEnabled=true
    Module.Param3WriteEnabled=true
   )
  ]
  rendererClasses [&NiagaraSpriteRendererProperties]
  renderers [{rendererClass=&NiagaraSpriteRendererProperties}]
 }
 Sparks {
  simTarget=GPUComputeSim
  emitterUpdateScript [
   &EmitterScalability(
    Module.bApplySpawnCountScale=false
    Module.bApplySpawnCountScaleByDistanceFraction=false
    Module.bUseMaxDistance=false
    Module.bUseMinDistance=false
   ) {enabled=false}
   &EmitterLifeCycle(NextLoopDuration=4.4208) {enabled=false}
   &EmitterState(Enable Distance Culling=true, MaxDistance=4000, Module.bUseMaxDistance=true, Module.bUseMinDistance=false, Scalability Mode=::NewEnumerator1)
   &SpawnBurst_Instantaneous(Spawn Count=150, Use Spawn Probability=false)
   &SpawnRate(
    SpawnRate=&FloatFromCurve(
     FloatCurve=DataInterface() {propertyValues {Curve {keys [<at -0.0046 1.0113>,<at 0.6325 0.0329>,<at 1.0011 0.002>]}}}
     CurveIndex=$Emitter.NormalizedLoopAge
     Scale Curve=24
    )
    Use Spawn Probability=false
   ) {enabled=false}
  ]
  particleSpawnScript [
   &InitializeParticle(
    Color=<rgba 50 0 26.1442 1>
    Lifetime=&UniformRangedFloat(Minimum=0.2, Maximum=2)
    Module.Write Color=true
    Module.Write Lifetime=true
    Module.Write Mass=true
    Module.Write Material Random=true
    Module.Write Position=false
    Module.Write Scale=false
    Module.Write SpriteRotation=false
    Module.Write SpriteSize=true
    Position=&SimulationPosition()
    Sprite Size=&Vector2DFromFloat(Value=&UniformRangedFloat(Minimum=2, Maximum=4))
   )
   &SampleStaticMesh(OverrideRandomSeed=false, StaticMesh:NiagaraDataInterfaceStaticMesh=$User.CrystalStaticMesh)
   &StaticMeshLocation(Module.Enable Axis Constraint=false)
   &AddVelocityInCone(Cone Angle=180, Module.Use Velocity Falloff On Cone Axis=false, Velocity Strength=100)
   &AddVelocityFromPoint(
    Module.Use Falloff Distance=false
    Origin Offset=<0 0 -100>
    Velocity Origin:Vector3f=&SimulationPosition()
    Velocity Strength=&UniformRangedFloat(Minimum=700, Maximum=1200)
   )
   &ApplyInitialForces(Apply Rotational Force to Rotational Velocity=false)
  ]
  particleUpdateScript [
   &ParticleState()
   &GravityForce(Gravity=<0 0 -1600>)
   &Drag(Use Linear Drag=true, Use Rotational Drag=false)
   &VectorNoiseForce(
    Force Amount=&FloatFromCurve(
     FloatCurve=DataInterface() {propertyValues {Curve {keys [{arriveTangent=-4.2476, interpMode=RCIM_Cubic, leaveTangent=-2.6975, tangentMode=RCTM_User, time=0, value=1}, <at 0.6 0>]}}}
     Scale Curve=3000
    )
    Module.Use Random Noise Time Interval=false
   )
   &SolveForcesAndVelocity(Clamp Velocity=false, Limit Acceleration=false)
   &ScaleColor(
    Scale RGB=&VectorFromCurve(
     VectorCurve=DataInterface() {propertyValues {
       XCurve {keys [
         {arriveTangent=-0.8547, interpMode=RCIM_Cubic, leaveTangent=-0.8547, tangentMode=RCTM_User, time=0, value=0.2}
         {arriveTangent=2.4102, interpMode=RCIM_Cubic, leaveTangent=2.4102, time=0.3199, value=0.6588}
         {arriveTangent=15.1384, interpMode=RCIM_Cubic, leaveTangent=15.1384, tangentMode=RCTM_User, time=0.3764, value=1.1071}
         {interpMode=RCIM_Cubic, tangentMode=RCTM_User, time=0.4, value=1.7898}
         {interpMode=RCIM_Cubic, leaveTangent=-6.5745, time=1, value=0.04}
        ]}
       YCurve {keys [{arriveTangent=-10.3118, interpMode=RCIM_Cubic, time=0, value=1}, <at 0.3199 0.3>, <at 0.4 0.2>]}
       ZCurve {keys [
         {arriveTangent=-1.8429, interpMode=RCIM_Cubic, time=0, value=2}
         {arriveTangent=-9.4313, interpMode=RCIM_Cubic, leaveTangent=-9.4313, tangentMode=RCTM_User, time=0.2232, value=1.2228}
         {arriveTangent=-4.9396, interpMode=RCIM_Cubic, leaveTangent=-4.9396, tangentMode=RCTM_User, time=0.3199, value=0.45}
         <at 0.4 0.2>
        ]}
      }}
     Scale Curve=&VectorFromFloat(Value=1)
    )
    ScaleA=false
    ScaleRGB=true
   )
   &ScaleSpriteSize(Scale Factor=&Vector2DFromFloat(Value=&FloatFromCurve(FloatCurve=DataInterface() {propertyValues {Curve {keys [<at 0.4 1 RCIM_Cubic>,<at 1.0011 0.002>]}}})))
   &ScaleSpriteSizeBySpeed(
    Max Scale Factor=<1.5 4>
    Sample Scale Factor By Curve=true
    Scale Factor Curve=DataInterface() {propertyValues {Curve {keys [
        {arriveTangent=0.2977, interpMode=RCIM_Cubic, time=0, value=0}
        {arriveTangent=0.7665, interpMode=RCIM_Cubic, leaveTangent=0.7665, tangentMode=RCTM_User, time=0.6496, value=0.2205}
        {arriveTangent=4.616, interpMode=RCIM_Cubic, leaveTangent=4.616, tangentMode=RCTM_User, time=1, value=1}
       ]}}}
    Velocity Threshold=700
   )
  ]
  rendererClasses [&NiagaraSpriteRendererProperties]
  renderers [{rendererClass=&NiagaraSpriteRendererProperties}]
 }
}
summary {
 systemName=NS_DesertPoint_Tree_Destruction_M
 emitters [
  {emitterName=DirectionalBurstMeshes_01, rendererClasses [&NiagaraMeshRendererProperties]}
  {emitterName=Pop_Sparks, simTarget=GPUComputeSim, rendererClasses [&NiagaraSpriteRendererProperties]}
  {emitterName=SlurpWispySmoke, rendererClasses [&NiagaraSpriteRendererProperties]}
  {emitterName=Sparks, simTarget=GPUComputeSim, rendererClasses [&NiagaraSpriteRendererProperties]}
  {emitterName=SlurpWispySmoke001, rendererClasses [&NiagaraSpriteRendererProperties]}
  {emitterName=HT_Debris, simTarget=GPUComputeSim, rendererClasses [&NiagaraSpriteRendererProperties]}
 ]
 userVariables [
  User.CrystalStaticMesh {
   type=NiagaraDataInterfaceStaticMesh
   description=""
   defaultValue {
    struct=NiagaraExt_VariableValue_DataInterface
    value {
     dataInterface=&self:NiagaraDataInterfaceStaticMesh_0
     dataInterfaceClass=&NiagaraDataInterfaceStaticMesh
    }
   }
  }
  User.CrystalWorldScale {type=Vector3f, description="", defaultValue {struct=Vector3f, value=<1 1 1>}}
 ]
}
